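; Blueprint for a lux ceiling light prop, built on a parent with an adjustable light and a fixture component.
; Its construction script saves the fixture's slot 0 material in a local variable, runs the parent construction script, then reapplies the saved material if valid.
; This preserves the fixture material through whatever the parent construction does.
; BeginPlay, overlap and Tick events are empty.
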(event EventBeginPlay)

(event Collision|EventActorBeginOverlap
  (params (OtherActor :type Actor%20Object%20Reference)))

(event EventTick
  (params (DeltaSeconds :type Float%20%28single-precision%29)))

(fn ConstructionScript
  (bind _l_fixture_mic (LocalVariables|GetL_Fixture_MIC))
  (LocalVariables|SetL_Fixture_MIC (Rendering|Material|GetMaterial (Variables|AdjustableLight|GetLightFixture) 0))
  (|CallParentFunction)
  (if (Utilities|IsValid _l_fixture_mic :!class /Script/Engine.KismetSystemLibrary)
    (Rendering|Material|SetMaterial (Variables|AdjustableLight|GetLightFixture) 0 _l_fixture_mic)))
